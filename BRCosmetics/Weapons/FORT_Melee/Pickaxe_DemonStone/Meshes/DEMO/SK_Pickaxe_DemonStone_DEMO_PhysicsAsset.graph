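# Physics asset for the DemonStone pickaxe demo skeletal mesh, a Fortnite cosmetic two-handed melee weapon.
# It holds a single body on the Melee_TwoHanded_Handle bone, made of one capsule that is offset and rotated from the bone.
# It has no constraints, so no joints link bodies and the capsule serves as the mesh's only simple physics shape.
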PhysicsAsset /BRCosmetics/Weapons/FORT_Melee/Pickaxe_DemonStone/Meshes/DEMO/SK_Pickaxe_DemonStone_DEMO_PhysicsAsset
bodies {Melee_TwoHanded_Handle [{shapeType=Capsule, length=68.5469, radius=42.5542, center=<0.8914 -7.5902 28.4985>, rotation=<rot 4.5186 -90 -0.1462>}]}
constraints []
